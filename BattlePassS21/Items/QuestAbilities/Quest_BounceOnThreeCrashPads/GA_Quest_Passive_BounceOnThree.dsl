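; Passive quest ability that tracks the player bouncing on three different Crash Pads (AppleSun) without landing.
; On activation it listens on the owning pawn for Event.AppleSunUsed, adds each instigating AppleSun actor to a unique array and stores the count.
; At three or more it sends a custom stat event to the player's quest manager.
; Removal of the MovementMode.Falling tag clears the array and count; debug prints report both count and reset.
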(event Ability|EventOnEndAbility
  (params (bWasCancelled :type Boolean))
  (Ability|EndAbility))

(event Ability|EventFailedToActivatePassiveAbility
  (params (ActorInfo :type Gameplay%20Ability%20Actor%20Info%20Structure))
  (Ability|EndAbility))

(event Ability|EventActivateAbility
  (bind _apple_sun_bgas (Variables|Default|GetApple%20Sun%20BGAs))
  (|CallParentFunction)
  (bind (_asynctaskproxy _payload) (Ability|Async|WaitGameplayEventToActor (Variables|GAVars|Data|GetOwningPlayerPawn) "(TagName=\"Event.AppleSunUsed\")")
    (:then
      (Ability|Tasks|WaitGameplayTagRemove "(TagName=\"MovementMode.Falling\")"
        (:Removed
          (Variables|Default|SetUnique%20Bounces 0)
          (Utilities|Array|Clear (Variables|Default|GetApple%20Sun%20BGAs))
          (Development|PrintString "Count Reset"))))
    (:EventReceived
      (if (.bSuccess (Utilities|Casting|CastToBGA_AppleSun_Apple_Athena (.Instigator (Utilities|Struct|BreakGameplayEventData _payload))))
        (Utilities|Array|AddUnique _apple_sun_bgas (Utilities|Casting|CastToBGA_AppleSun_Apple_Athena (.Instigator (Utilities|Struct|BreakGameplayEventData _payload))))
        (bind _output_get (Variables|Default|SetUnique%20Bounces (Utilities|Array|Length _apple_sun_bgas)))
        (Development|PrintString (Utilities|String|ToString(Integer) _output_get) true true "(R=0.000000,G=0.660000,B=1.000000,A=1.000000)" 5.0)
        (if (>= _output_get 3)
          (Quest|SendComplexCustomStatEvent (Variables|GAVars|Data|GetOwningPlayerQuestManager) 0 (Variables|Default|GetAdditional%20Source%20Tags) "(GameplayTags=)" 1))))))
